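; Gameplay ability that reports lilypad and tire bounces to the quest system as quest progress.
; On event activation it calls the parent, switches on the event's gameplay tag, sends a matching complex custom stat event to the owning player's quest manager, then ends.
; It also defines two helpers unused by that event: a signed-direction 360-degree vector angle calculation, and CalculateDegreesTurned, which only sets Max Speed.
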(event Ability|EventActivateAbilityFromEvent
  (params (EventData :type Gameplay%20Event%20Data%20Structure%20%28by%20ref%29))
  (bind _owningplayerquestmanager (Variables|GAVars|Data|GetOwningPlayerQuestManager))
  (|CallParentFunction EventData)
  (switch Utilities|FlowControl|Switch|SwitchonGameplayTag (Utilities|Struct|BreakGameplayEventData EventData)
    (:Default
      _)
    (:Event.Lilypad.Bounce
      (Quest|SendComplexCustomStatEvent _owningplayerquestmanager 0 "(GameplayTags=((TagName=\"Event.Lilypad.Bounce\")))" "(GameplayTags=)" 1))
    (:Event.Tire.Bounce
      (Quest|SendComplexCustomStatEvent _owningplayerquestmanager 0 "(GameplayTags=((TagName=\"Event.Tire.Bounce\")))" "(GameplayTags=)" 1)))
  (Ability|EndAbility))

(fn AngleBetweenVectors
  (params (Destination%20Rotation :type Vector) (Origin%20Rotation :type Vector) (Rotation%20Normal :type Vector))
  (return (Math|Float|SelectFloat (Math|Trig|RadiansToDegrees (Math|Trig|Acos(Radians) (/ (Math|Vector|DotProduct Origin%20Rotation Destination%20Rotation) (* (Math|Vector|VectorLength Destination%20Rotation) (Math|Vector|VectorLength Origin%20Rotation))))) (- 3.6e+02 (Math|Trig|RadiansToDegrees (Math|Trig|Acos(Radians) (/ (Math|Vector|DotProduct Origin%20Rotation Destination%20Rotation) (* (Math|Vector|VectorLength Destination%20Rotation) (Math|Vector|VectorLength Origin%20Rotation)))))) (> (Math|Vector|DotProduct Rotation%20Normal (Math|Vector|CrossProduct Origin%20Rotation Destination%20Rotation)) 0))))

(fn CalculateDegreesTurned
  (params (A :type Float%20%28double-precision%29) (B :type Float%20%28double-precision%29))
  (Variables|Default|SetMax%20Speed A))
